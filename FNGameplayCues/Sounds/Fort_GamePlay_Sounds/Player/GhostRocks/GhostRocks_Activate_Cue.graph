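# GhostRocks_Activate_Cue is a SoundCue for the GhostRocks activation sound, stored under player gameplay sounds.
# A single WavePlayer plays GhostRocks_Activate_01 through a Modulator node left at default settings.
# The cue routes to the Fort_Gadgets_Items sound class with reduced volume and uses the GhostRocks attenuation asset for spatial falloff.
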SoundCue /FNGameplayCues/Sounds/Fort_GamePlay_Sounds/Player/GhostRocks/GhostRocks_Activate_Cue
root=Modulator_0
nodes [
 {id=Modulator_0, children [WavePlayer_0]}
 {id=WavePlayer_0, properties {soundWaveAssetPtr=/Game/Sounds/Fort_GamePlay_Sounds/Player/GhostRocks/GhostRocks_Activate_01}}
]
properties {
 attenuationSettings=/Game/Sounds/Attenuation/Fort_GhostRocks_Att
 maxDistance=4500
 soundClassObject=&Fort_Gadgets_Items
 volumeMultiplier=0.6
}
