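# Looping sound cue for Ghost Rocks in use by a player, spatialized with the Ghost Rocks attenuation and capped by a stop-quietest concurrency override.
# A mixer layers the in-use loop, the in-world loop and a Cube shoreline loop, the last at boosted input.
# A Velocity parameter modulates the in-use loop's pitch and volume and the shoreline's volume, while a looping envelope wavers the in-world loop.
# Quality switches drop the in-world layer or substitute an optimized shoreline wave.
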SoundCue /FNGameplayCues/Sounds/Fort_GamePlay_Sounds/Player/GhostRocks/GhostRocks_InUse_Loop_Cue
root=Mixer_0
nodes [
 {
  id=Enveloper_0
  children [WavePlayer_0]
  properties {
   bLoop=true
   bLoopIndefinitely=true
   loopEnd=20
   pitchCurve {editorCurveData {keys [{time=0}, <at 10 0.8>, <at 15 1.1>, {time=20}]}}
   volumeCurve {editorCurveData {keys [<at 0 1>,<at 10 0.75>,<at 15 1>]}}
  }
 }
 {id=Mixer_0, children [ModulatorContinuous_0, QualityLevel_0, ModulatorContinuous_1], properties {inputVolume [1,1,2]}}
 {
  id=ModulatorContinuous_0
  children [WavePlayer_1]
  properties {
   pitchModulationParams {default=0, maxOutput=1.25, minOutput=0.8, parameterName=Velocity}
   volumeModulationParams {default=0, minOutput=0.5, parameterName=Velocity}
  }
 }
 {
  id=ModulatorContinuous_1
  children [QualityLevel_1]
  properties {pitchModulationParams {maxOutput=1.25, minOutput=0.8}, volumeModulationParams {default=0, parameterName=Velocity}}
 }
 {id=QualityLevel_0, children [Enveloper_0,Enveloper_0,null,null]}
 {id=QualityLevel_1, children [WavePlayer_4,WavePlayer_4,WavePlayer_3,WavePlayer_3]}
 {id=WavePlayer_0, properties {bLooping=true, soundWaveAssetPtr=/Game/Sounds/Fort_GamePlay_Sounds/Player/GhostRocks/GhostRocks_InWorldLoop_01}}
 {
  id=WavePlayer_1
  properties {bLooping=true, soundWaveAssetPtr=&/GhostRocks_InUseLoop_01}
 }
 {id=WavePlayer_3, properties {bLooping=true, soundWaveAssetPtr=/FNGameplayCues/Athena/Sounds/Events/Cube/Destination/Cube_Shoreline_01_opt}}
 {id=WavePlayer_4, properties {bLooping=true, soundWaveAssetPtr=/FNGameplayCues/Athena/Sounds/Events/Cube/Destination/Cube_Shoreline_01}}
]
properties {
 attenuationSettings=/Game/Sounds/Attenuation/Fort_GhostRocks_Att
 bOverrideConcurrency=true
 maxDistance=4500
 soundClassObject=&Fort_Gadgets_Items
 volumeMultiplier=0.5
 concurrencyOverrides {maxCount=8, resolutionRule=StopQuietest}
}
